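; Passive Battle Pass S21 quest ability that credits eliminations made with the Prime (Overload-tagged) shotgun's first shot.
; A Killed event listener infers first shot from current magazine count, then records DBNO victims or sends the quest stat immediately.
; An EarnedElimination listener sends the stat when the target matches a recorded downed player.
; That match sets a flag the graph never clears, so later eliminations also send the stat.
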(event Ability|EventOnEndAbility
  (params (bWasCancelled :type Boolean))
  (Ability|EndAbility))

(event Ability|EventFailedToActivatePassiveAbility
  (params (ActorInfo :type Gameplay%20Ability%20Actor%20Info%20Structure))
  (Ability|EndAbility))

(event Ability|EventActivateAbility
  (|CallParentFunction)
  (Variables|Default|SetPlayers%20Downed%20by%201st%20Prime (Variables|Default|GetPlayers%20Downed%20by%201st%20Prime))
  (Utilities|FlowControl|Sequence
    (:then_0
      (bind (_asynctaskproxy _payload) (Ability|Tasks|WaitGameplayEvent "(TagName=\"Event.Damage.Killed\")")
        (:EventReceived
          (if (.bSuccess (Utilities|Casting|CastToFortPlayerPawnAthena (.Target (Utilities|Struct|BreakGameplayEventData _payload))))
            (Variables|Default|SetNewly%20Downed%20Player (Utilities|Casting|CastToFortPlayerPawnAthena (.Target (Utilities|Struct|BreakGameplayEventData _payload))))
            (bind _condition (CallFunction|FirstShotOfPrimeCheck))
            (if _condition
              (if (Health|IsDBNO (Variables|Default|GetNewly%20Downed%20Player))
                (Utilities|Array|AddUnique (Variables|Default|GetPlayers%20Downed%20by%201st%20Prime) (Variables|Default|GetNewly%20Downed%20Player))
                (else
                  (Quest|SendComplexCustomStatEvent (Variables|GAVars|Data|GetOwningPlayerQuestManager) 0 "(GameplayTags=((TagName=\"Athena.Quests.PrimeElim\")))" "(GameplayTags=)" 1))))))))
    (:then_1
      (bind (_asynctaskproxy_2 _payload_2) (Ability|Tasks|WaitGameplayEvent "(TagName=\"Event.EarnedElimination\")")
        (:EventReceived
          (for _array_element (Variables|Default|GetPlayers%20Downed%20by%201st%20Prime)
            (if (== _array_element (Utilities|Casting|CastToFortPlayerPawnAthena (.Target (Utilities|Struct|BreakGameplayEventData _payload_2))))
              (Variables|Default|SetSame%20downed%20player true)))
          (if (Variables|Default|GetSame%20downed%20player)
            (Quest|SendComplexCustomStatEvent (Variables|GAVars|Data|GetOwningPlayerQuestManager) 0 "(GameplayTags=((TagName=\"Athena.Quests.PrimeElim\")))" "(GameplayTags=)" 1)))))))

(fn FirstShotofPrimeCheck
  (bind _owningplayerpawn (Variables|GAVars|Data|GetOwningPlayerPawn))
  (bind _returnvalue (Weapon|GetCurrentWeapon _owningplayerpawn))
  (Variables|Default|SetMax%20Ammo (FortWeapon|GetBulletsPerClip _returnvalue))
  (return (and (Tags|WeaponDataHasMatchingGameplayTag _returnvalue "(TagName=\"Item.Weapon.Ranged.Shotgun.Overload\")") (>= (FortWeapon|GetMagazineAmmoCount _returnvalue) (- (Variables|Default|GetMax%20Ammo) 1)))))
