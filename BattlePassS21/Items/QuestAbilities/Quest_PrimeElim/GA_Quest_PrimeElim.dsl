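; Gameplay ability that credits the PrimeElim quest stat for Overload shotgun eliminations made with the magazine at most one round below full.
; On activation it records the target pawn, filters by eliminatable-player tags, then checks weapon and ammo.
; Qualifying non-DBNO hits score immediately; DBNO targets await an EarnedElimination event matched against the downed-player list.
; Failed weapon checks remove the target, then recheck weapon and ammo upon EarnedElimination.
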(event Ability|EventActivateAbilityFromEvent
  (params (EventData :type Gameplay%20Event%20Data%20Structure%20%28by%20ref%29))
  (bind (_eventtag _instigator _target _optionalobject _optionalobject2 _contexthandle _instigatortags _targettags) (Utilities|Struct|BreakGameplayEventData EventData))
  (bind _owningplayerpawn (Variables|GAVars|Data|GetOwningPlayerPawn))
  (bind _returnvalue (Weapon|GetCurrentWeapon _owningplayerpawn))
  (bind _owningplayerpawn_2 (Variables|GAVars|Data|GetOwningPlayerPawn))
  (bind _returnvalue_2 (Weapon|GetCurrentWeapon _owningplayerpawn_2))
  (bind (_asfort_player_pawn_athena _bsuccess) (Utilities|Casting|CastToFortPlayerPawnAthena _target))
  (|CallParentFunction EventData)
  (Variables|Default|SetEvent%20Tag _eventtag)
  (Utilities|Array|AddUnique (Variables|Default|GetDowned%20Players) _asfort_player_pawn_athena)
  (if _bsuccess
    (if (GameplayTags|HasAnyTags _targettags (Variables|Default|GetEliminatable%20Player%20Tags))
      (Variables|Default|SetMax%20Ammo (FortWeapon|GetBulletsPerClip _returnvalue))
      (if (and (Tags|WeaponDataHasMatchingGameplayTag _returnvalue "(TagName=\"Item.Weapon.Ranged.Shotgun.Overload\")") (>= (FortWeapon|GetMagazineAmmoCount _returnvalue) (- (Variables|Default|GetMax%20Ammo) 1)))
        (if (Health|IsDBNO _asfort_player_pawn_athena)
          (bind (_asynctaskproxy _payload) (Ability|Tasks|WaitGameplayEvent "(TagName=\"Event.EarnedElimination\")")
            (:EventReceived
              (for _array_element (Variables|Default|GetDowned%20Players)
                (if (== _array_element (Utilities|Casting|CastToFortPlayerPawnAthena (.Target (Utilities|Struct|BreakGameplayEventData _payload))))
                  (Variables|Default|SetSame%20downed%20player true)))
              (if (Variables|Default|GetSame%20downed%20player)
                (Quest|SendComplexCustomStatEvent (Variables|GAVars|Data|GetOwningPlayerQuestManager) 0 "(GameplayTags=((TagName=\"Athena.Quests.PrimeElim\")))" "(GameplayTags=)" 1)
                (else
                  (Ability|EndAbility)))))
          (else
            (Quest|SendComplexCustomStatEvent (Variables|GAVars|Data|GetOwningPlayerQuestManager) 0 "(GameplayTags=((TagName=\"Athena.Quests.PrimeElim\")))" "(GameplayTags=)" 1)))
        (else
          (Utilities|Array|RemoveItem (Variables|Default|GetDowned%20Players) _asfort_player_pawn_athena)
          (Ability|Tasks|WaitGameplayEvent "(TagName=\"Event.EarnedElimination\")"
            (:EventReceived
              (Variables|Default|SetMax%20Ammo (FortWeapon|GetBulletsPerClip _returnvalue_2))
              (if (and (Tags|WeaponDataHasMatchingGameplayTag _returnvalue_2 "(TagName=\"Item.Weapon.Ranged.Shotgun.Overload\")") (>= (FortWeapon|GetMagazineAmmoCount _returnvalue_2) (- (Variables|Default|GetMax%20Ammo) 1)))
                (Quest|SendComplexCustomStatEvent (Variables|GAVars|Data|GetOwningPlayerQuestManager) 0 "(GameplayTags=((TagName=\"Athena.Quests.PrimeElim\")))" "(GameplayTags=)" 1)
                (else
                  (Ability|EndAbility)))))))
      (else
        _)))
  (Ability|EndAbility))
